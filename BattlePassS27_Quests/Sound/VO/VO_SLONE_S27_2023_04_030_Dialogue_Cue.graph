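# SoundCue for a Slone voice-over line in the Season 27 Battle Pass quests, set up as audio-log dialogue.
# A single Dialogue Player root node plays the matching Dialogue Wave, using the AudioLogVO_DV speaker with no targets.
# It applies the shared audio-log attenuation, routes through the narrator VO sound class and raises the volume multiplier above default.
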SoundCue /BattlePassS27_Quests/Sound/VO/VO_SLONE_S27_2023_04_030_Dialogue_Cue
root=DialoguePlayer_0
nodes [{
  id=DialoguePlayer_0
  properties {dialogueWaveParameter {
    dialogueWave=/BattlePassS27_Quests/Sound/VO/Waves/VO_SLONE_S27_2023_04_030_Dialogue
    context {speaker=/BattlePassS27_Quests/Sound/AudioLogVO_DV, targets []}
   }}
 }]
properties {
 attenuationSettings=/BattlePassS27_Quests/Sound/Fort_AudioLogs_Att
 maxDistance=2097152
 soundClassObject=/Game/Sounds/FortSoundClassesAndModes/Fort_VO_Narrator
 volumeMultiplier=2
}
